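# M_ColorWheel is an unlit masked helper material in the CustomMipGenerator tool that tints cloud noise with an adjustable hue.
# The Phase parameter sets the hue of a fully saturated, full-value HSV color, which is converted to RGB.
# That color multiplies an atmospheric cloud noise texture softened by negative CheapContrast.
# The red channel of the adjusted noise is also split out, and the mask clip value is zero.
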Material /TechArtTools/CustomMipGenerator/HelperMaterials/M_ColorWheel
graph {
 expressions {
  MaterialExpressionComponentMask_0 {class=MaterialExpressionComponentMask, inputs [None {from=MaterialExpressionMaterialFunctionCall_0, output=Result}], properties {r=true}}
  MaterialExpressionConstant_0 {class=MaterialExpressionConstant, properties {r=-0.2}}
  MaterialExpressionConstant_1 {class=MaterialExpressionConstant, properties {r=1}}
  MaterialExpressionConvert_0 {
   class=MaterialExpressionConvert
   inputs [
    None {from=MaterialExpressionScalarParameter_0}
    None {from=""}
    None {from=""}
   ]
   properties {
    nodeName=Make Vector3
    convertInputs [
     {
      type=Scalar
      expressionInput {
       expression=&self:MaterialExpressionScalarParameter_0
       inputName=None
       mask=0
       maskA=0
       maskB=0
       maskG=0
       maskR=0
       outputIndex=0
      }
     }
     {
      type=Scalar
      defaultValue=<rgba 1 0 0 1>
      expressionInput {expression=null, inputName=None, mask=0, maskA=0, maskB=0, maskG=0, maskR=0, outputIndex=0}
     }
     {
      type=Scalar
      defaultValue=<rgba 1 0 0 1>
      expressionInput {expression=null, inputName=None, mask=0, maskA=0, maskB=0, maskG=0, maskR=0, outputIndex=0}
     }
    ]
    convertOutputs [{type=Vector3}]
   }
  }
  MaterialExpressionHsvToRgb_0 {class=MaterialExpressionHsvToRgb, inputs [None {from=MaterialExpressionConvert_0, output=""}], properties {}}
  MaterialExpressionMaterialFunctionCall_0 {
   class=MaterialExpressionMaterialFunctionCall
   inputs [
    In {from=MaterialExpressionTextureSample_0, output=RGB}
    Contrast {from=MaterialExpressionConstant_0}
   ]
   properties {materialFunction=&CheapContrast}
  }
  MaterialExpressionMultiply_0 {
   class=MaterialExpressionMultiply
   inputs [
    A {from=MaterialExpressionHsvToRgb_0}
    B {from=MaterialExpressionMaterialFunctionCall_0, output=Result}
   ]
  }
  MaterialExpressionScalarParameter_0 {class=MaterialExpressionScalarParameter, properties {parameterName=Phase}}
  MaterialExpressionTextureSample_0 {
   class=MaterialExpressionTextureSample
   inputs [
    UVs {from=""}
    Tex {from=""}
    Apply View MipBias {from=""}
   ]
   properties {texture=&AtmosphericCloudNoise01}
  }
 }
 properties {
  bAutomaticallySetUsageInEditor=false
  bUsedWithEditorCompositing=true
  blendMode=BLEND_Masked
  opacityMaskClipValue=0
  shadingModel=MSM_Unlit
  usedShadingModels=Unlit
 }
}
