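# UI-domain helper material for the Custom Mip Generator tool that displays a chosen mip level of a texture parameter.
# The Loop parameter blends the mip level from zero to rounded time, wrapped by the Modulo parameter, so the view steps through mips over time.
# Loop also blends in a DebugScalarValues readout that adds the current mip number over the texture.
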Material /TechArtTools/CustomMipGenerator/HelperMaterials/M_DisplayTexture
graph {
 properties {bAutomaticallySetUsageInEditor=false, materialDomain=MD_UI, usedShadingModels=Default Lit}
 nodes [
  n1 = Time()
  n2 = Modulo!(
   A=LinearInterpolate(A=Constant(), B=Round(None=@n1), Alpha=ScalarParameter() {parameterName=Loop}) {constAlpha=1}
   B=ScalarParameter() {defaultValue=12, parameterName=Modulo}
  )
  n3 = TextureSampleParameter2D!(UVs, Level=@n2, Apply View MipBias) {mipValueMode=TMVM_MipLevel, parameterName=DisplayMip, texture=&/T_TechArt}
  n4 = Constant() {r=2}
  n5 = Constant4Vector() {constant=<rgba 0.05 0.05 0.35 0.125>}
  LinearInterpolate(
   A=@n3.RGB
   B=Add(
    A=@n3.RGB
    B=MaterialFunctionCall(Number=@n2, MaximumNumberOfDigits=@n4, DebugTextLocation RG_UpperRight BA_LowerLeft=@n5.RGBA) {materialFunction=&DebugScalarValues}.Result
   )
   Alpha=ScalarParameter() {parameterName=Loop}
  ) {constAlpha=1}
 ]
}
