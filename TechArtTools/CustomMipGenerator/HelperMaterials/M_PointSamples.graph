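# Unlit masked helper material for the Custom Mip Generator tool that point-samples a texture supplied through the TextureToRender parameter.
# The engine's PointSampledUVs function uses that texture's own size to snap the UVs for unfiltered texel reads.
# The sample uses an explicit mip level left at zero, so it reads the top mip regardless of screen-space derivatives.
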Material /TechArtTools/CustomMipGenerator/HelperMaterials/M_PointSamples
graph {
 properties {
  bAutomaticallySetUsageInEditor=false
  bUsedWithEditorCompositing=true
  blendMode=BLEND_Masked
  opacityMaskClipValue=0
  shadingModel=MSM_Unlit
  usedShadingModels=Unlit
 }
 nodes [
  n1 = TextureObjectParameter() {parameterName=TextureToRender, texture=&/T_TechArt}
  TextureSample!(
   UVs=MaterialFunctionCall!(UVs=TextureCoordinate(), Texture Size=TextureProperty(Tex=@n1)) {materialFunction=/Engine/Functions/Engine_MaterialFunctions01/Texturing/PointSampledUVs}.Result
   Tex=@n1
   Level=Constant()
   Apply View MipBias
  ) {mipValueMode=TMVM_MipLevel, texture=&AtmosphericCloudNoise01}
 ]
}
