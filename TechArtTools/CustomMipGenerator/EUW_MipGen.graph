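# Editor utility widget demonstrating custom mip generation by packing render targets into saved Texture2D assets as hand-authored mip levels.
# Generate draws a phase-shifted color wheel into ten fixed targets (512 to 1), previewed beside the resulting texture.
# Generate from Textures point-samples an editable texture list into matching-size targets, using GetFormat for format and temporarily overriding source compression.
# Clear blanks the fixed targets and resets the preview.
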EditorUtilityWidgetBlueprint /TechArtTools/CustomMipGenerator/EUW_MipGen
graphs [
 EventGraph {nodes [
   "On Clicked (EditorUtilityButton_0)"() {is=title K2Node_ComponentBoundEvent}
   > n1 = GetEditorWorld(self=Get UnrealEditorSubsystem())
   n2 = Make Array!(
    "[0]":Texture Render Target 2D Object Reference=&/RT_512
    "[1]":Texture Render Target 2D Object Reference=&/RT_256
    "[2]":Texture Render Target 2D Object Reference=&/RT_128
    "[3]":Texture Render Target 2D Object Reference=&/RT_64
    "[4]":Texture Render Target 2D Object Reference=&/RT_32
    "[5]":Texture Render Target 2D Object Reference=&/RT_16
    "[6]":Texture Render Target 2D Object Reference=&/RT_8
    "[7]":Texture Render Target 2D Object Reference=&/RT_4
    "[8]":Texture Render Target 2D Object Reference=&/RT_2
    "[9]":Texture Render Target 2D Object Reference=&/RT_1
   ) -> (Array:Array of Texture Render Target 2D Object References)
   > n3 = For Each Loop(Array:Array of Texture Render Target 2D Object References=@n2) -> (Array Element:Texture Render Target 2D Object Reference)
   n4 = MakeLiteralString(Value=&/T_CustomGeneratedMipTextureFromRTs)
   n5 = Create Texture Asset from Render Target Mips!(
    execute=@n3.Completed
    WorldContextObject:Object Reference=@n1
    RenderTargets:Array of Texture Render Target 2D Object References=@n2
    PackagePathAndName:String=@n4
   ) -> (then, bSuccess:bool, OutErrorMessage:String, ReturnValue:Texture 2D Object Reference)
   > Set CompressionNoAlpha!(execute, CompressionNoAlpha:bool=true, self:Texture Object Reference=@n5.ReturnValue) -> (then, Output_Get:bool)
   > Set NeverStream!(execute, NeverStream:bool=false, self:Streamable Render Asset Object Reference=@n5.ReturnValue) -> (then, Output_Get:bool)
   > SaveLoadedAsset(AssetToSave=@n5.ReturnValue, bOnlyIfIsDirty=false)
   > n6 = DoesAssetExist(AssetPath=@n4)
   > Branch(Condition=@n6)
   > n7 = LoadAsset(AssetPath=@n4)
   > n8 = Cast To Texture2D(Object=@n7)
   n9 = Get MID2() -> (MID2:Material Instance Dynamic Object Reference)
   > SetTextureParameterValue(self=@n9, ParameterName=DisplayMip, Value=@n8)
   > SetScalarParameterValue(self=@n9, ParameterName=Loop, Value=1)
   > SetScalarParameterValue(self=@n9, ParameterName=Modulo, Value=10)
   n10 = float / float(
    A="To Float (Integer)"(InInt=@n3.Array Index)
    B="To Float (Integer)"(InInt=Length(TargetArray:Array of Texture Render Target 2D Object References=@n2))
   )
   SetScalarParameterValue(execute=@n3.LoopBody, self=Get MID() -> (MID:Material Instance Dynamic Object Reference), ParameterName=Phase, Value=@n10)
   > DrawMaterialToRenderTarget!(
    execute
    WorldContextObject:Object Reference=@n1
    TextureRenderTarget:Texture Render Target 2D Object Reference="Get (a copy)"(Array:Array of Texture Render Target 2D Object References=@n2, Dimension 1=@n3.Array Index) -> (Output:Texture Render Target 2D Object Reference)
    Material:Material Interface Object Reference=Get MID()
   ) -> (then)
   "On Clicked (Clear)"() {is=title K2Node_ComponentBoundEvent}
   > n11 = For Each Loop(Array=Make Array(
     "[0]"=&RT_512
     "[1]"=&RT_256
     "[2]"=&RT_128
     "[3]"=&RT_64
     "[4]"=&RT_32
     "[5]"=&RT_16
     "[6]"=&RT_8
     "[7]"=&RT_4
     "[8]"=&RT_2
     "[9]"=&RT_1
    ))
   > n12 = GetEditorWorld(self=Get UnrealEditorSubsystem())
   > ClearRenderTarget2D!(
    execute
    WorldContextObject:Object Reference=@n12
    TextureRenderTarget:Texture Render Target 2D Object Reference=@n11
    ClearColor:Linear Color Structure=(R=0,G=0,B=0,A=1)
   ) -> (then)
   n13 = Get MID2()
   > SetTextureParameterValue(self=@n13, ParameterName=DisplayMip, Value=/TechArtTools/CustomMipGenerator/HelperMaterials/T_TechArt)
   > SetScalarParameterValue(self=@n13, ParameterName=Loop, Value=0)
   "On Clicked (GenFromAssets)"() {is=title K2Node_ComponentBoundEvent}
   > Clear(TargetArray:Array of Texture Render Target 2D Object References=Get RTs from File() -> (RTs from File:Array of Texture Render Target 2D Object References))
   n14 = Get Textures() -> (Textures:Array of Texture 2D Object References)
   n15 = "Get (a copy)"(Array:Array of Texture 2D Object References=@n14) -> (Output:Texture 2D Object Reference)
   > n16 = GetFormat!(execute, self, self2:Texture Object Reference=@n15) -> (
    then
    sRGB:bool
    texture render target format1:ETextureRenderTargetFormat Enum=RTF_R8
    Original Compression Settings:TextureCompressionSettings Enum
    Temp Compression for render:TextureCompressionSettings Enum
   )
   > Set Compress Without Alpha(Compress Without Alpha:bool=Get CompressionNoAlpha!(self:Texture Object Reference=@n15) -> (CompressionNoAlpha:bool)) -> (Output_Get:bool)
   > Set Temp Compression For Render(Temp Compression For Render:TextureCompressionSettings Enum=@n16.Temp Compression for render) -> (Output_Get:TextureCompressionSettings Enum)
   > Set Format(Format:ETextureRenderTargetFormat Enum=@n16.texture render target format1) -> (Output_Get:ETextureRenderTargetFormat Enum=RTF_R8)
   > Set Original Compression Settings(Original Compression Settings:TextureCompressionSettings Enum=@n16.Original Compression Settings) -> (Output_Get:TextureCompressionSettings Enum)
   > n17 = GetEditorWorld(self=Get UnrealEditorSubsystem())
   > n18 = For Each Loop(Array:Array of Texture 2D Object References=@n14) -> (Array Element:Texture 2D Object Reference)
   n19 = MakeLiteralString(Value=&/T_CustomGeneratedMipTextureFromFile)
   n20 = Create Texture Asset from Render Target Mips(execute=@n18.Completed, WorldContextObject=@n17, RenderTargets=Get RTs from File(), PackagePathAndName=@n19)
   > n21 = Branch(Condition=@n20)
   > SaveLoadedAsset(AssetToSave=@n20.ReturnValue, bOnlyIfIsDirty=false)
   > n22 = DoesAssetExist(AssetPath=@n19)
   > Branch(Condition=@n22)
   > n23 = LoadAsset(AssetPath=@n19)
   > n24 = Cast To Texture2D(Object=@n23)
   n25 = Get MID2()
   > SetTextureParameterValue(self=@n25, ParameterName=DisplayMip, Value=@n24)
   > SetScalarParameterValue(self=@n25, ParameterName=Loop, Value=1)
   > SetScalarParameterValue(self=@n25, ParameterName=Modulo, Value="To Float (Integer)"(InInt=Length(TargetArray:Array of Texture 2D Object References=@n14)))
   PrintString(execute=@n21.else, InString=@n20.OutErrorMessage)
   n26 = CreateRenderTarget2D!(
    execute=@n18.LoopBody
    WorldContextObject:Object Reference=@n17
    Width:int=GetSizeX!(self:Texture 2D Object Reference=@n18) -> (ReturnValue:int=0)
    Height:int=GetSizeY!(self:Texture 2D Object Reference=@n18) -> (ReturnValue:int=0)
    Format:ETextureRenderTargetFormat Enum=Get Format() -> (Format:ETextureRenderTargetFormat Enum=RTF_R8)
    ClearColor:Linear Color Structure=(R=0,G=0,B=0,A=1)
    bAutoGenerateMipMaps:bool=false
    bSupportUAVs:bool=false
   ) -> (then, ReturnValue:Texture Render Target 2D Object Reference)
   > Set SRGB(SRGB=@n16, self=@n26)
   > Set CompressionSettings(CompressionSettings=TC_Default, self=@n18)
   > SetTextureParameterValue(self=Get MID3() -> (MID3:Material Instance Dynamic Object Reference), ParameterName=TextureToRender, Value=@n18)
   > DrawMaterialToRenderTarget(WorldContextObject=@n17, TextureRenderTarget=@n26, Material=Get MID3())
   > Add(
    TargetArray:Array of Texture Render Target 2D Object References=Get RTs from File()
    NewItem:"Texture Render Target 2D Object Reference (by ref)"=@n26
   ) {is=title K2Node_CallArrayFunction}
   > Set CompressionSettings(CompressionSettings=Get Original Compression Settings() -> (Original Compression Settings:TextureCompressionSettings Enum), self=@n18)
   Event Construct()
   > n27 = CreateDynamicMaterialInstance(Parent=/TechArtTools/CustomMipGenerator/HelperMaterials/M_ColorWheel, OptionalName=None, CreationFlags=Transient)
   > Set MID(MID:Material Instance Dynamic Object Reference=@n27) -> (Output_Get:Material Instance Dynamic Object Reference)
   n28 = Self-Reference()
   > SetObject(self=Get DetailsView_60() -> (DetailsView_60:Details View Object Reference), NewObject=@n28)
   > SetObject(self=Get DetailsView() -> (DetailsView:Details View Object Reference), NewObject=@n28)
   > n29 = CreateDynamicMaterialInstance(Parent=/TechArtTools/CustomMipGenerator/HelperMaterials/M_DisplayTexture, OptionalName=None, CreationFlags=Transient)
   > Set MID2(MID2:Material Instance Dynamic Object Reference=@n29) -> (Output_Get:Material Instance Dynamic Object Reference)
   > SetBrushFromMaterial(self=Get Image() -> (Image:Image Object Reference), Material=Get MID2())
   > n30 = CreateDynamicMaterialInstance(Parent=/TechArtTools/CustomMipGenerator/HelperMaterials/M_PointSamples, OptionalName=None, CreationFlags=Transient)
   > Set MID3(MID3:Material Instance Dynamic Object Reference=@n30) -> (Output_Get:Material Instance Dynamic Object Reference)
   HSV to RGB!(H:float=0, S:float=0, V:float=0, A:float=1) -> (ReturnValue:Linear Color Structure)
   PrintString(InString="To String (Float)"(InDouble=@n10))
   CreateRenderTarget2D(Width=256, Height=256, Format=RTF_RGBA16f, ClearColor=(R=0,G=0,B=0,A=1), bAutoGenerateMipMaps=false, bSupportUAVs=false)
   MakeColor!(R:float=0, G:float=0, B:float=0, A:float=1) -> (ReturnValue:Linear Color Structure)
   Event PreConstruct()
   Event Tick()
  ]}
 GetFormat {nodes [
   n1 = GetFormat!() -> (then, self2:Texture Object Reference) {is=title K2Node_FunctionEntry}
   > n2 = Sequence()
   n3 = Get CompressionSettings(self=@n1)
   > n4 = Branch(Condition=OR Boolean!(
     A="Equal (Enum)"(A:TextureCompressionSettings Enum=@n3, B:TextureCompressionSettings Enum=TC_Default)
     B="Equal (Enum)"(A=@n3, B=TC_Masks)
     C:bool="Equal (Enum)"(A=@n3, B=TC_VectorDisplacementmap)
     D:bool="Equal (Enum)"(A=@n3, B=TC_EditorIcon)
    ) -> (ReturnValue))
   > Set texture render target format(texture render target format:ETextureRenderTargetFormat Enum=RTF_RGBA8) -> (Output_Get:ETextureRenderTargetFormat Enum=RTF_R8)
   > Set Temp Comp for render(Temp Comp for render:TextureCompressionSettings Enum=TC_EditorIcon) -> (Output_Get:TextureCompressionSettings Enum)
   n5 = Branch(execute=@n4.else, Condition=OR Boolean!(A="Equal (Enum)"(A=@n3, B=TC_HDR), B="Equal (Enum)"(A=@n3, B=TC_HDR_Compressed)) -> (ReturnValue))
   > Set texture render target format(texture render target format=RTF_RGBA16f) -> (Output_Get=RTF_R8)
   > Set Temp Comp for render(Temp Comp for render=TC_HDR)
   Branch(execute=@n5.else, Condition="Equal (Enum)"(A=@n3, B=TC_HDR_F32))
   > Set texture render target format(texture render target format=RTF_RGBA32f) -> (Output_Get=RTF_R8)
   > Set Temp Comp for render(Temp Comp for render=TC_HDR_F32)
   Return Node!(
    execute=@n2.then_1
    sRGB:bool=Get SRGB!(self:Texture Object Reference=@n1) -> (SRGB:bool)
    texture render target format1:ETextureRenderTargetFormat Enum=Get texture render target format() -> (texture render target format:ETextureRenderTargetFormat Enum=RTF_R8)
    Original Compression Settings:TextureCompressionSettings Enum=@n3
    Temp Compression for render:TextureCompressionSettings Enum=Get Temp Comp for render() -> (Temp Comp for render:TextureCompressionSettings Enum)
   )
  ]}
]
variables [
 Compress Without Alpha {type=boolean}
 Format {type=ETextureRenderTargetFormat, default=RTF_R8}
 MID {type=MaterialInstanceDynamic}
 MID2 {type=MaterialInstanceDynamic}
 MID3 {type=MaterialInstanceDynamic}
 Original Compression Settings {type=TextureCompressionSettings, default=TC_Default}
 RTs from File {type="array<TextureRenderTarget2D>", default []}
 Temp Compression For Render {type=TextureCompressionSettings, default=TC_Default}
 Textures {
  type="array<Texture2D>"
  default [
   /TechArtTools/CustomMipGenerator/TextureFileToMips/Mip0
   /TechArtTools/CustomMipGenerator/TextureFileToMips/Mip1
   /TechArtTools/CustomMipGenerator/TextureFileToMips/Mip2
   /TechArtTools/CustomMipGenerator/TextureFileToMips/Mip3
  ]
 }
]
widgets {
 info {rootWidgetClass=&VerticalBox, widgetCount=54}
 widgets [
  VerticalBox_232 {widgetClassPath=&VerticalBox, widget=VerticalBox_232, properties {bOverride_Cursor=true}}
  VerticalBox_82 {
   widgetClassPath=&VerticalBox
   parent=VerticalBox_232
   widget=VerticalBox_82
   slot {class=VerticalBoxSlot, properties {horizontalAlignment=HAlign_Left, padding=<ltrb 32 32 0 0>, size {sizeRule=Fill, value=1}}}
  }
  TextBlock_528 {
   widgetClassPath=&TextBlock
   parent=VerticalBox_82
   widget=TextBlock_528
   properties {
    text=Custom Mip Generation Demonstration
    font {
     bForceMonospaced=false
     bMaterialIsStencil=false
     fontMaterial=null
     fontObject=/Engine/EngineFonts/RobotoTiny
     letterSpacing=32
     monospacedWidth=1
     size=24
     skewAmount=0
     typefaceFontName=Light
     outlineSettings {
      bApplyOutlineToDropShadows=false
      bMiteredCorners=false
      bSeparateFillAlpha=false
      outlineMaterial=null
      outlineSize=0
      outlineColor=<rgba 0 0 0 1>
     }
    }
   }
   slot {class=VerticalBoxSlot}
  }
  Spacer_406 {widgetClassPath=&Spacer, parent=VerticalBox_82, widget=Spacer_406, properties {size=<0 32>}, slot {class=VerticalBoxSlot}}
  HorizontalBox_163 {widgetClassPath=&HorizontalBox, parent=VerticalBox_82, widget=HorizontalBox_163, slot {class=VerticalBoxSlot}}
  UniformGridPanel_96 {widgetClassPath=/Script/UMG.UniformGridPanel, parent=HorizontalBox_163, widget=UniformGridPanel_96, slot {class=HorizontalBoxSlot}}
  Overlay_134 {
   widgetClassPath=&Overlay
   parent=UniformGridPanel_96
   widget=Overlay_134
   slot {class=UniformGridSlot, properties {horizontalAlignment=HAlign_Fill, verticalAlignment=VAlign_Fill}}
  }
  Image_0 {
   widgetClassPath=&Image
   bIsVariable=true
   parent=Overlay_134
   widget=Image_0
   properties {brush {
     bIsDynamicallyLoaded=false
     drawAs=Image
     imageType=FullColor
     mirroring=NoMirror
     resourceName=None
     resourceObject=&RT_512
     tiling=NoTile
     imageSize=<128 128>
     margin=<ltrb 0 0 0 0>
     outlineSettings {
      bUseBrushTransparency=false
      roundingType=HalfHeightRadius
      width=0
      color {colorUseRule=UseColor_Specified, specifiedColor=<rgba 0 0 0 0>}
      cornerRadii=<0 0 0 0>
     }
     tintColor {colorUseRule=UseColor_Specified, specifiedColor=<rgba 1 1 1 1>}
     uVRegion {bIsValid=false, max=<0 0>, min=<0 0>}
    }}
   slot {class=OverlaySlot, properties {horizontalAlignment=HAlign_Fill, verticalAlignment=VAlign_Fill, padding=<ltrb 4 4 4 4>}}
  }
  TextBlock_404 {
   widgetClassPath=&TextBlock
   parent=Overlay_134
   widget=TextBlock_404
   properties {
    text=Mip 0
    font {
     bForceMonospaced=false
     bMaterialIsStencil=false
     fontMaterial=null
     fontObject=&RobotoTiny
     letterSpacing=0
     monospacedWidth=1
     size=18
     skewAmount=0
     typefaceFontName=Light
     outlineSettings {
      bApplyOutlineToDropShadows=false
      bMiteredCorners=false
      bSeparateFillAlpha=false
      outlineMaterial=null
      outlineSize=0
      outlineColor=<rgba 0 0 0 1>
     }
    }
   }
   slot {class=OverlaySlot, properties {verticalAlignment=VAlign_Bottom, padding=<ltrb 6 0 0 0>}}
  }
  Overlay {widgetClassPath=&Overlay, parent=UniformGridPanel_96, widget=Overlay, slot {class=UniformGridSlot, properties {column=1}}}
  Image_1 {
   widgetClassPath=&Image
   bIsVariable=true
   parent=Overlay
   widget=Image_1
   properties {brush {
     bIsDynamicallyLoaded=false
     drawAs=Image
     imageType=FullColor
     mirroring=NoMirror
     resourceName=None
     resourceObject=&RT_256
     tiling=NoTile
     imageSize=<128 128>
     margin=<ltrb 0 0 0 0>
     outlineSettings {
      bUseBrushTransparency=false
      roundingType=HalfHeightRadius
      width=0
      color {colorUseRule=UseColor_Specified, specifiedColor=<rgba 0 0 0 0>}
      cornerRadii=<0 0 0 0>
     }
     tintColor {colorUseRule=UseColor_Specified, specifiedColor=<rgba 1 1 1 1>}
     uVRegion {bIsValid=false, max=<0 0>, min=<0 0>}
    }}
   slot {class=OverlaySlot, properties {horizontalAlignment=HAlign_Fill, verticalAlignment=VAlign_Fill, padding=<ltrb 4 4 4 4>}}
  }
  TextBlock {
   widgetClassPath=&TextBlock
   parent=Overlay
   widget=TextBlock
   properties {
    text=Mip 1
    font {
     bForceMonospaced=false
     bMaterialIsStencil=false
     fontMaterial=null
     fontObject=&RobotoTiny
     letterSpacing=0
     monospacedWidth=1
     size=18
     skewAmount=0
     typefaceFontName=Light
     outlineSettings {
      bApplyOutlineToDropShadows=false
      bMiteredCorners=false
      bSeparateFillAlpha=false
      outlineMaterial=null
      outlineSize=0
      outlineColor=<rgba 0 0 0 1>
     }
    }
   }
   slot {class=OverlaySlot, properties {verticalAlignment=VAlign_Bottom, padding=<ltrb 6 0 0 0>}}
  }
  Overlay_3 {widgetClassPath=&Overlay, parent=UniformGridPanel_96, widget=Overlay_3, slot {class=UniformGridSlot, properties {column=2}}}
  Image_2 {
   widgetClassPath=&Image
   bIsVariable=true
   parent=Overlay_3
   widget=Image_2
   properties {brush {
     bIsDynamicallyLoaded=false
     drawAs=Image
     imageType=FullColor
     mirroring=NoMirror
     resourceName=None
     resourceObject=&RT_128
     tiling=NoTile
     imageSize=<128 128>
     margin=<ltrb 0 0 0 0>
     outlineSettings {
      bUseBrushTransparency=false
      roundingType=HalfHeightRadius
      width=0
      color {colorUseRule=UseColor_Specified, specifiedColor=<rgba 0 0 0 0>}
      cornerRadii=<0 0 0 0>
     }
     tintColor {colorUseRule=UseColor_Specified, specifiedColor=<rgba 1 1 1 1>}
     uVRegion {bIsValid=false, max=<0 0>, min=<0 0>}
    }}
   slot {class=OverlaySlot, properties {horizontalAlignment=HAlign_Fill, verticalAlignment=VAlign_Fill, padding=<ltrb 4 4 4 4>}}
  }
  TextBlock_3 {
   widgetClassPath=&TextBlock
   parent=Overlay_3
   widget=TextBlock_3
   properties {
    text=Mip 2
    font {
     bForceMonospaced=false
     bMaterialIsStencil=false
     fontMaterial=null
     fontObject=&RobotoTiny
     letterSpacing=0
     monospacedWidth=1
     size=18
     skewAmount=0
     typefaceFontName=Light
     outlineSettings {
      bApplyOutlineToDropShadows=false
      bMiteredCorners=false
      bSeparateFillAlpha=false
      outlineMaterial=null
      outlineSize=0
      outlineColor=<rgba 0 0 0 1>
     }
    }
   }
   slot {class=OverlaySlot, properties {verticalAlignment=VAlign_Bottom, padding=<ltrb 6 0 0 0>}}
  }
  Overlay_4 {widgetClassPath=&Overlay, parent=UniformGridPanel_96, widget=Overlay_4, slot {class=UniformGridSlot, properties {row=1}}}
  Image_3 {
   widgetClassPath=&Image
   bIsVariable=true
   parent=Overlay_4
   widget=Image_3
   properties {brush {
     bIsDynamicallyLoaded=false
     drawAs=Image
     imageType=FullColor
     mirroring=NoMirror
     resourceName=None
     resourceObject=&RT_64
     tiling=NoTile
     imageSize=<128 128>
     margin=<ltrb 0 0 0 0>
     outlineSettings {
      bUseBrushTransparency=false
      roundingType=HalfHeightRadius
      width=0
      color {colorUseRule=UseColor_Specified, specifiedColor=<rgba 0 0 0 0>}
      cornerRadii=<0 0 0 0>
     }
     tintColor {colorUseRule=UseColor_Specified, specifiedColor=<rgba 1 1 1 1>}
     uVRegion {bIsValid=false, max=<0 0>, min=<0 0>}
    }}
   slot {class=OverlaySlot, properties {horizontalAlignment=HAlign_Fill, verticalAlignment=VAlign_Fill, padding=<ltrb 4 4 4 4>}}
  }
  TextBlock_4 {
   widgetClassPath=&TextBlock
   parent=Overlay_4
   widget=TextBlock_4
   properties {
    text=Mip 3
    font {
     bForceMonospaced=false
     bMaterialIsStencil=false
     fontMaterial=null
     fontObject=&RobotoTiny
     letterSpacing=0
     monospacedWidth=1
     size=18
     skewAmount=0
     typefaceFontName=Light
     outlineSettings {
      bApplyOutlineToDropShadows=false
      bMiteredCorners=false
      bSeparateFillAlpha=false
      outlineMaterial=null
      outlineSize=0
      outlineColor=<rgba 0 0 0 1>
     }
    }
   }
   slot {class=OverlaySlot, properties {verticalAlignment=VAlign_Bottom, padding=<ltrb 6 0 0 0>}}
  }
  Overlay_5 {
   widgetClassPath=&Overlay
   parent=UniformGridPanel_96
   widget=Overlay_5
   slot {class=UniformGridSlot, properties {column=1, row=1}}
  }
  Image_4 {
   widgetClassPath=&Image
   bIsVariable=true
   parent=Overlay_5
   widget=Image_4
   properties {brush {
     bIsDynamicallyLoaded=false
     drawAs=Image
     imageType=FullColor
     mirroring=NoMirror
     resourceName=None
     resourceObject=&RT_32
     tiling=NoTile
     imageSize=<128 128>
     margin=<ltrb 0 0 0 0>
     outlineSettings {
      bUseBrushTransparency=false
      roundingType=HalfHeightRadius
      width=0
      color {colorUseRule=UseColor_Specified, specifiedColor=<rgba 0 0 0 0>}
      cornerRadii=<0 0 0 0>
     }
     tintColor {colorUseRule=UseColor_Specified, specifiedColor=<rgba 1 1 1 1>}
     uVRegion {bIsValid=false, max=<0 0>, min=<0 0>}
    }}
   slot {class=OverlaySlot, properties {horizontalAlignment=HAlign_Fill, verticalAlignment=VAlign_Fill, padding=<ltrb 4 4 4 4>}}
  }
  TextBlock_5 {
   widgetClassPath=&TextBlock
   parent=Overlay_5
   widget=TextBlock_5
   properties {
    text=Mip 4
    font {
     bForceMonospaced=false
     bMaterialIsStencil=false
     fontMaterial=null
     fontObject=&RobotoTiny
     letterSpacing=0
     monospacedWidth=1
     size=18
     skewAmount=0
     typefaceFontName=Light
     outlineSettings {
      bApplyOutlineToDropShadows=false
      bMiteredCorners=false
      bSeparateFillAlpha=false
      outlineMaterial=null
      outlineSize=0
      outlineColor=<rgba 0 0 0 1>
     }
    }
   }
   slot {class=OverlaySlot, properties {verticalAlignment=VAlign_Bottom}}
  }
  Overlay_6 {
   widgetClassPath=&Overlay
   parent=UniformGridPanel_96
   widget=Overlay_6
   slot {class=UniformGridSlot, properties {column=2, row=1}}
  }
  Image_5 {
   widgetClassPath=&Image
   bIsVariable=true
   parent=Overlay_6
   widget=Image_5
   properties {brush {
     bIsDynamicallyLoaded=false
     drawAs=Image
     imageType=FullColor
     mirroring=NoMirror
     resourceName=None
     resourceObject=&RT_16
     tiling=NoTile
     imageSize=<128 128>
     margin=<ltrb 0 0 0 0>
     outlineSettings {
      bUseBrushTransparency=false
      roundingType=HalfHeightRadius
      width=0
      color {colorUseRule=UseColor_Specified, specifiedColor=<rgba 0 0 0 0>}
      cornerRadii=<0 0 0 0>
     }
     tintColor {colorUseRule=UseColor_Specified, specifiedColor=<rgba 1 1 1 1>}
     uVRegion {bIsValid=false, max=<0 0>, min=<0 0>}
    }}
   slot {class=OverlaySlot, properties {horizontalAlignment=HAlign_Fill, verticalAlignment=VAlign_Fill, padding=<ltrb 4 4 4 4>}}
  }
  TextBlock_6 {
   widgetClassPath=&TextBlock
   parent=Overlay_6
   widget=TextBlock_6
   properties {
    text=Mip 5
    font {
     bForceMonospaced=false
     bMaterialIsStencil=false
     fontMaterial=null
     fontObject=&RobotoTiny
     letterSpacing=0
     monospacedWidth=1
     size=18
     skewAmount=0
     typefaceFontName=Light
     outlineSettings {
      bApplyOutlineToDropShadows=false
      bMiteredCorners=false
      bSeparateFillAlpha=false
      outlineMaterial=null
      outlineSize=0
      outlineColor=<rgba 0 0 0 1>
     }
    }
   }
   slot {class=OverlaySlot, properties {verticalAlignment=VAlign_Bottom, padding=<ltrb 6 0 0 0>}}
  }
  Overlay_7 {widgetClassPath=&Overlay, parent=UniformGridPanel_96, widget=Overlay_7, slot {class=UniformGridSlot, properties {row=2}}}
  Image_6 {
   widgetClassPath=&Image
   bIsVariable=true
   parent=Overlay_7
   widget=Image_6
   properties {brush {
     bIsDynamicallyLoaded=false
     drawAs=Image
     imageType=FullColor
     mirroring=NoMirror
     resourceName=None
     resourceObject=&RT_8
     tiling=NoTile
     imageSize=<128 128>
     margin=<ltrb 0 0 0 0>
     outlineSettings {
      bUseBrushTransparency=false
      roundingType=HalfHeightRadius
      width=0
      color {colorUseRule=UseColor_Specified, specifiedColor=<rgba 0 0 0 0>}
      cornerRadii=<0 0 0 0>
     }
     tintColor {colorUseRule=UseColor_Specified, specifiedColor=<rgba 1 1 1 1>}
     uVRegion {bIsValid=false, max=<0 0>, min=<0 0>}
    }}
   slot {class=OverlaySlot, properties {horizontalAlignment=HAlign_Fill, verticalAlignment=VAlign_Fill, padding=<ltrb 4 4 4 4>}}
  }
  TextBlock_7 {
   widgetClassPath=&TextBlock
   parent=Overlay_7
   widget=TextBlock_7
   properties {
    text=Mip 6
    font {
     bForceMonospaced=false
     bMaterialIsStencil=false
     fontMaterial=null
     fontObject=&RobotoTiny
     letterSpacing=0
     monospacedWidth=1
     size=18
     skewAmount=0
     typefaceFontName=Light
     outlineSettings {
      bApplyOutlineToDropShadows=false
      bMiteredCorners=false
      bSeparateFillAlpha=false
      outlineMaterial=null
      outlineSize=0
      outlineColor=<rgba 0 0 0 1>
     }
    }
   }
   slot {class=OverlaySlot, properties {verticalAlignment=VAlign_Bottom, padding=<ltrb 6 0 0 0>}}
  }
  Overlay_8 {
   widgetClassPath=&Overlay
   parent=UniformGridPanel_96
   widget=Overlay_8
   slot {class=UniformGridSlot, properties {column=1, row=2}}
  }
  Image_7 {
   widgetClassPath=&Image
   bIsVariable=true
   parent=Overlay_8
   widget=Image_7
   properties {brush {
     bIsDynamicallyLoaded=false
     drawAs=Image
     imageType=FullColor
     mirroring=NoMirror
     resourceName=None
     resourceObject=&RT_4
     tiling=NoTile
     imageSize=<128 128>
     margin=<ltrb 0 0 0 0>
     outlineSettings {
      bUseBrushTransparency=false
      roundingType=HalfHeightRadius
      width=0
      color {colorUseRule=UseColor_Specified, specifiedColor=<rgba 0 0 0 0>}
      cornerRadii=<0 0 0 0>
     }
     tintColor {colorUseRule=UseColor_Specified, specifiedColor=<rgba 1 1 1 1>}
     uVRegion {bIsValid=false, max=<0 0>, min=<0 0>}
    }}
   slot {class=OverlaySlot, properties {horizontalAlignment=HAlign_Fill, verticalAlignment=VAlign_Fill, padding=<ltrb 4 4 4 4>}}
  }
  TextBlock_8 {
   widgetClassPath=&TextBlock
   parent=Overlay_8
   widget=TextBlock_8
   properties {
    text=Mip 7
    font {
     bForceMonospaced=false
     bMaterialIsStencil=false
     fontMaterial=null
     fontObject=&RobotoTiny
     letterSpacing=0
     monospacedWidth=1
     size=18
     skewAmount=0
     typefaceFontName=Light
     outlineSettings {
      bApplyOutlineToDropShadows=false
      bMiteredCorners=false
      bSeparateFillAlpha=false
      outlineMaterial=null
      outlineSize=0
      outlineColor=<rgba 0 0 0 1>
     }
    }
   }
   slot {class=OverlaySlot, properties {verticalAlignment=VAlign_Bottom, padding=<ltrb 6 0 0 0>}}
  }
  Overlay_9 {
   widgetClassPath=&Overlay
   parent=UniformGridPanel_96
   widget=Overlay_9
   slot {class=UniformGridSlot, properties {column=2, row=2}}
  }
  Image_8 {
   widgetClassPath=&Image
   bIsVariable=true
   parent=Overlay_9
   widget=Image_8
   properties {brush {
     bIsDynamicallyLoaded=false
     drawAs=Image
     imageType=FullColor
     mirroring=NoMirror
     resourceName=None
     resourceObject=&RT_2
     tiling=NoTile
     imageSize=<128 128>
     margin=<ltrb 0 0 0 0>
     outlineSettings {
      bUseBrushTransparency=false
      roundingType=HalfHeightRadius
      width=0
      color {colorUseRule=UseColor_Specified, specifiedColor=<rgba 0 0 0 0>}
      cornerRadii=<0 0 0 0>
     }
     tintColor {colorUseRule=UseColor_Specified, specifiedColor=<rgba 1 1 1 1>}
     uVRegion {bIsValid=false, max=<0 0>, min=<0 0>}
    }}
   slot {class=OverlaySlot, properties {horizontalAlignment=HAlign_Fill, verticalAlignment=VAlign_Fill, padding=<ltrb 4 4 4 4>}}
  }
  TextBlock_9 {
   widgetClassPath=&TextBlock
   parent=Overlay_9
   widget=TextBlock_9
   properties {
    text=Mip 8
    font {
     bForceMonospaced=false
     bMaterialIsStencil=false
     fontMaterial=null
     fontObject=&RobotoTiny
     letterSpacing=0
     monospacedWidth=1
     size=18
     skewAmount=0
     typefaceFontName=Light
     outlineSettings {
      bApplyOutlineToDropShadows=false
      bMiteredCorners=false
      bSeparateFillAlpha=false
      outlineMaterial=null
      outlineSize=0
      outlineColor=<rgba 0 0 0 1>
     }
    }
   }
   slot {class=OverlaySlot, properties {verticalAlignment=VAlign_Bottom, padding=<ltrb 6 0 0 0>}}
  }
  Overlay_1 {widgetClassPath=&Overlay, parent=UniformGridPanel_96, widget=Overlay_1, slot {class=UniformGridSlot, properties {row=3}}}
  Image_9 {
   widgetClassPath=&Image
   bIsVariable=true
   parent=Overlay_1
   widget=Image_9
   properties {brush {
     bIsDynamicallyLoaded=false
     drawAs=Image
     imageType=FullColor
     mirroring=NoMirror
     resourceName=None
     resourceObject=&RT_1
     tiling=NoTile
     imageSize=<128 128>
     margin=<ltrb 0 0 0 0>
     outlineSettings {
      bUseBrushTransparency=false
      roundingType=HalfHeightRadius
      width=0
      color {colorUseRule=UseColor_Specified, specifiedColor=<rgba 0 0 0 0>}
      cornerRadii=<0 0 0 0>
     }
     tintColor {colorUseRule=UseColor_Specified, specifiedColor=<rgba 1 1 1 1>}
     uVRegion {bIsValid=false, max=<0 0>, min=<0 0>}
    }}
   slot {class=OverlaySlot, properties {horizontalAlignment=HAlign_Fill, verticalAlignment=VAlign_Fill, padding=<ltrb 4 4 4 4>}}
  }
  TextBlock_1 {
   widgetClassPath=&TextBlock
   parent=Overlay_1
   widget=TextBlock_1
   properties {
    text=Mip 9
    font {
     bForceMonospaced=false
     bMaterialIsStencil=false
     fontMaterial=null
     fontObject=&RobotoTiny
     letterSpacing=0
     monospacedWidth=1
     size=18
     skewAmount=0
     typefaceFontName=Light
     outlineSettings {
      bApplyOutlineToDropShadows=false
      bMiteredCorners=false
      bSeparateFillAlpha=false
      outlineMaterial=null
      outlineSize=0
      outlineColor=<rgba 0 0 0 1>
     }
    }
   }
   slot {class=OverlaySlot, properties {verticalAlignment=VAlign_Bottom, padding=<ltrb 6 0 0 0>}}
  }
  Spacer_177 {
   widgetClassPath=&Spacer
   parent=HorizontalBox_163
   widget=Spacer_177
   properties {size=<79.9166 1>}
   slot {class=HorizontalBoxSlot}
  }
  Image {
   widgetClassPath=&Image
   bIsVariable=true
   parent=HorizontalBox_163
   widget=Image
   properties {brush {
     bIsDynamicallyLoaded=false
     drawAs=Image
     imageType=FullColor
     mirroring=NoMirror
     resourceName=None
     resourceObject=&M_DisplayTexture
     tiling=NoTile
     imageSize=<512 512>
     margin=<ltrb 0 0 0 0>
     outlineSettings {
      bUseBrushTransparency=false
      roundingType=HalfHeightRadius
      width=0
      color {colorUseRule=UseColor_Specified, specifiedColor=<rgba 0 0 0 0>}
      cornerRadii=<0 0 0 0>
     }
     tintColor {colorUseRule=UseColor_Specified, specifiedColor=<rgba 1 1 1 1>}
     uVRegion {bIsValid=false, max=<0 0>, min=<0 0>}
    }}
   slot {class=HorizontalBoxSlot}
  }
  DetailsView_60 {
   widgetClassPath=&DetailsView
   parent=VerticalBox_82
   widget=DetailsView_60
   properties {bIsEnabled=false, visibility=Hidden, propertiesToShow [MID]}
   slot {class=VerticalBoxSlot}
  }
  Spacer_326 {widgetClassPath=&Spacer, parent=VerticalBox_82, widget=Spacer_326, properties {size=<1 24>}, slot {class=VerticalBoxSlot}}
  HorizontalBox_279 {widgetClassPath=&HorizontalBox, parent=VerticalBox_82, widget=HorizontalBox_279, slot {class=VerticalBoxSlot}}
  Spacer_1 {widgetClassPath=&Spacer, parent=HorizontalBox_279, widget=Spacer_1, properties {size=<32 1>}, slot {class=HorizontalBoxSlot}}
  EditorUtilityButton_0 {
   widgetClassPath=&EditorUtilityButton
   parent=HorizontalBox_279
   widget=EditorUtilityButton_0
   slot {class=HorizontalBoxSlot}
  }
  TextBlock_0 {
   widgetClassPath=&TextBlock
   parent=EditorUtilityButton_0
   widget=TextBlock_0
   properties {
    text=Generate
    font {
     bForceMonospaced=false
     bMaterialIsStencil=false
     fontMaterial=null
     fontObject=&RobotoTiny
     letterSpacing=32
     monospacedWidth=1
     size=24
     skewAmount=0
     typefaceFontName=Light
     outlineSettings {
      bApplyOutlineToDropShadows=false
      bMiteredCorners=false
      bSeparateFillAlpha=false
      outlineMaterial=null
      outlineSize=0
      outlineColor=<rgba 0 0 0 1>
     }
    }
   }
   slot {class=ButtonSlot}
  }
  Spacer {widgetClassPath=&Spacer, parent=HorizontalBox_279, widget=Spacer, properties {size=<32 1>}, slot {class=HorizontalBoxSlot}}
  Clear {widgetClassPath=&EditorUtilityButton, parent=HorizontalBox_279, widget=Clear, slot {class=HorizontalBoxSlot}}
  TextBlock_2 {
   widgetClassPath=&TextBlock
   parent=Clear
   widget=TextBlock_2
   properties {
    text=Clear Render Targets
    font {
     bForceMonospaced=false
     bMaterialIsStencil=false
     fontMaterial=null
     fontObject=&RobotoTiny
     letterSpacing=32
     monospacedWidth=1
     size=24
     skewAmount=0
     typefaceFontName=Light
     outlineSettings {
      bApplyOutlineToDropShadows=false
      bMiteredCorners=false
      bSeparateFillAlpha=false
      outlineMaterial=null
      outlineSize=0
      outlineColor=<rgba 0 0 0 1>
     }
    }
   }
   slot {class=ButtonSlot}
  }
  Spacer_2 {widgetClassPath=&Spacer, parent=HorizontalBox_279, widget=Spacer_2, properties {size=<32 1>}, slot {class=HorizontalBoxSlot}}
  GenFromAssets {
   widgetClassPath=&EditorUtilityButton
   parent=HorizontalBox_279
   widget=GenFromAssets
   slot {class=HorizontalBoxSlot, properties {horizontalAlignment=HAlign_Left}}
  }
  TextBlock_10 {
   widgetClassPath=&TextBlock
   parent=GenFromAssets
   widget=TextBlock_10
   properties {
    text=Generate from Textures
    font {
     bForceMonospaced=false
     bMaterialIsStencil=false
     fontMaterial=null
     fontObject=&RobotoTiny
     letterSpacing=32
     monospacedWidth=1
     size=24
     skewAmount=0
     typefaceFontName=Light
     outlineSettings {
      bApplyOutlineToDropShadows=false
      bMiteredCorners=false
      bSeparateFillAlpha=false
      outlineMaterial=null
      outlineSize=0
      outlineColor=<rgba 0 0 0 1>
     }
    }
   }
   slot {class=ButtonSlot}
  }
  EditorUtilityScrollBox_153 {
   widgetClassPath=&EditorUtilityScrollBox
   parent=VerticalBox_82
   widget=EditorUtilityScrollBox_153
   properties {alwaysShowScrollbar=true, alwaysShowScrollbarTrack=true}
   slot {class=VerticalBoxSlot, properties {size {sizeRule=Fill, value=1}}}
  }
  VerticalBox_323 {widgetClassPath=&VerticalBox, parent=EditorUtilityScrollBox_153, widget=VerticalBox_323, slot {class=ScrollBoxSlot}}
  DetailsView {
   widgetClassPath=&DetailsView
   parent=VerticalBox_323
   widget=DetailsView
   properties {propertiesToShow [Textures]}
   slot {class=VerticalBoxSlot, properties {verticalAlignment=VAlign_Top}}
  }
  HorizontalBox_393 {widgetClassPath=&HorizontalBox, parent=VerticalBox_323, widget=HorizontalBox_393, slot {class=VerticalBoxSlot}}
 ]
}
